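# Surface material for the Present glider cosmetic: a lit, skeletal-mesh material that blends base color toward a tint color through animated masks.
# Panning soft noise distorts bow and ribbon mask lookups, while procedural stripes from rotated, panning pre-skinned positions are cut by a streak texture.
# Normal, specular-mask and cropped-emissive inputs join that color in material attributes passed through the accessories tech-art override function.
# The additive speed-line color branch is left unconnected.
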Material /BRCosmetics/Gadgets/Assets/VinderTech_GliderChute/Glider_Present/Materials/M_GLider_Present
graph {
 properties {
  bUseMaterialAttributes=true
  bUsedWithMeshDeformer=true
  bUsedWithSkeletalMesh=true
  pixelDepthOffsetMode=PDOM_Legacy
  usedShadingModels=Default Lit
 }
 nodes [
  n1 = TextureCoordinate()
  n2 = ComponentMask(None=VectorParameter() {parameterName=BowNoise_Tiling, defaultValue=<rgba 1 1 1 1>}.RGB) {g=true, r=true}
  n3 = VectorParameter() {parameterName=BowSpeed, defaultValue=<rgba 0.5 0.5 0 1>}
  n4 = TextureSample(UVs=Panner(Coordinate=Multiply(A=@n1, B=@n2), Speed=ComponentMask(None=@n3.RGB) {g=true, r=true})) {texture=/Game/Characters/Player/Male/Medium/Bodies/M_MED_Viper/Textures/Glitch/T_SoftNoise}
  n5 = Subtract(A=Multiply(A=ComponentMask(None=@n4.RGB) {g=true, r=true}) {constB=2})
  n6 = TextureObjectParameter() {
   parameterName=MaskTexture_02
   samplerType=SAMPLERTYPE_Masks
   texture=/BRCosmetics/Gadgets/Assets/VinderTech_GliderChute/Glider_Present/Textures/T_Glider_Present_FX_2
  }
  n7 = TextureSample(Tex=@n6) {samplerType=SAMPLERTYPE_Masks, texture=&T_Glider_Present_FX_2}
  n8 = Multiply(A=ScalarParameter() {defaultValue=0.02, parameterName=Bow_MoveStrength}, B=@n7.G)
  n9 = TextureSample(UVs=Add(A=TextureCoordinate(), B=Multiply(A=@n5, B=@n8)), Tex=@n6) {samplerType=SAMPLERTYPE_Masks, texture=&T_Glider_Present_FX_2}
  n10 = TextureObjectParameter() {
   parameterName=MaskTexture
   samplerType=SAMPLERTYPE_Masks
   texture=/BRCosmetics/Gadgets/Assets/VinderTech_GliderChute/Glider_Present/Textures/T_Glider_Present_FX
  }
  n11 = TextureSample(Tex=@n10) {samplerType=SAMPLERTYPE_Masks, texture=&T_Glider_Present_FX}
  n12 = TextureCoordinate()
  n13 = ComponentMask(None=VectorParameter() {parameterName=Noise_Tiling, defaultValue=<rgba 1 1 1 1>}.RGB) {g=true, r=true}
  n14 = VectorParameter() {parameterName=Speed, defaultValue=<rgba 0.5 0.5 0 1>}
  n15 = TextureSample(UVs=Panner(Coordinate=Multiply(A=@n12, B=@n13), Speed=ComponentMask(None=@n14.RGB) {g=true, r=true})) {texture=&T_SoftNoise}
  n16 = Subtract(A=Multiply(A=ComponentMask(None=@n15.RGB) {g=true, r=true}) {constB=2})
  n17 = TextureSample(Tex=@n10) {samplerType=SAMPLERTYPE_Masks, texture=&T_Glider_Present_FX}
  n18 = Multiply(A=ScalarParameter() {defaultValue=0.02, parameterName=Ribbon_MoveStrength}, B=@n17.A)
  n19 = TextureSample(UVs=Add(A=TextureCoordinate(), B=Multiply(A=@n16, B=@n18)), Tex=@n10) {samplerType=SAMPLERTYPE_Masks, texture=&T_Glider_Present_FX}
  n20 = Max(A=@n11.R, B=@n19.G)
  n21 = ComponentMask(None=VertexInterpolator(VS=PreSkinnedPosition()).PS) {g=true, r=true}
  n22 = Rotator(Coordinate=@n21, Time=ScalarParameter() {defaultValue=-3.14, parameterName=LinesRotation})
  n23 = ComponentMask(None=Divide(A=Panner(Coordinate=@n22) {speedX=32}, B=ScalarParameter() {defaultValue=128, parameterName=LinesBaseScale}) {constB=32}) {r=true}
  n24 = Ceil(None=Subtract(A=Frac(None=@n23), B=ScalarParameter() {defaultValue=0.7, parameterName=Line_01Thickness}))
  n25 = Subtract(A=@n23, B=ScalarParameter() {defaultValue=0.2, parameterName=Line_02_Offset})
  n26 = Ceil(None=Subtract(A=Frac(None=@n25), B=ScalarParameter() {defaultValue=0.95, parameterName=Line_02_Thickness}))
  n27 = ScalarParameter() {defaultValue=0.5, parameterName=Line_03_Offset}
  n28 = Subtract(A=Frac(None=Subtract(A=@n23, B=@n27)), B=ScalarParameter() {defaultValue=0.9, parameterName=Line_03_Thickness})
  n29 = ScalarParameter() {defaultValue=0.5, parameterName=Line_04_Offset}
  n30 = Subtract(A=Frac(None=Subtract(A=@n23, B=@n29)), B=ScalarParameter() {defaultValue=0.97, parameterName=Line_04_Thickness})
  n31 = ScalarParameter() {defaultValue=0.5, parameterName=Line_05_Offset}
  n32 = Subtract(A=Frac(None=Subtract(A=@n23, B=@n31)), B=ScalarParameter() {defaultValue=0.97, parameterName=Line_05_Thickness})
  n33 = Max(A=Max(A=Max(A=@n24, B=@n26), B=Ceil(None=@n28)), B=Max(A=Ceil(None=@n30), B=Ceil(None=@n32)))
  n34 = Rotator(Coordinate=@n21, Time=ScalarParameter() {defaultValue=6.28, parameterName=StreaksRotation})
  n35 = ComponentMask(None=VectorParameter() {parameterName=LinesPanSpeed, defaultValue=<rgba 0 -32 0 1>}.RGB) {g=true, r=true}
  n36 = VectorParameter() {parameterName=SpeedLinesTiling, defaultValue=<rgba 32 32 32 1>}
  n37 = TextureSample(UVs=Divide(A=Panner(Coordinate=@n34, Speed=@n35), B=ComponentMask(None=@n36.RGB) {g=true, r=true})) {samplerType=SAMPLERTYPE_Masks, texture=/BRCosmetics/Gadgets/Assets/VinderTech_GliderChute/Glider_New_Year_2020/Textures/T_Blast}
  n38 = Multiply(A=@n17.B, B=Clamp(None=Subtract(A=@n33, B=@n37.R)))
  n39 = LinearInterpolate(
   A=TextureSampleParameter2D() {parameterName=BaseColor, texture=/BRCosmetics/Gadgets/Assets/VinderTech_GliderChute/Glider_Present/Textures/T_Glider_Present_D}.RGB
   B=VectorParameter() {parameterName=Tech Art Color, defaultValue=<rgba 0.013 0.0561 0.3916 1>}.RGB
   Alpha=Max(A=@n9.R, B=Max(A=@n20, B=@n38))
  )
  n40 = VectorParameter() {parameterName=SpeedLinesColor, defaultValue=<rgba 0.3564 0.3613 0.3564 1>}
  Add(A=@n39, B=Multiply(A=Multiply(A=@n37.R, B=@n40.RGB), B=@n17.B))
  n41 = TextureSampleParameter2D() {
   parameterName=SpecularMasks
   samplerType=SAMPLERTYPE_Masks
   texture=/BRCosmetics/Gadgets/Assets/VinderTech_GliderChute/Glider_Present/Textures/T_Glider_Present_S
  }
  n42 = TextureSampleParameter2D() {
   parameterName=Normal
   samplerType=SAMPLERTYPE_Normal
   texture=/BRCosmetics/Gadgets/Assets/VinderTech_GliderChute/Glider_Present/Textures/T_Glider_Present_N
  }
  n43 = SetMaterialAttributes!(
   MaterialAttributes
   Base Color=@n39
   Metallic=@n41.G
   Specular=@n41.G
   Roughness=@n41.G
   Emissive Color=MaterialFunctionCall() {materialFunction=/Game/Packages/Fortress_SharedMaterials/Base_Material_Functions/BaseCustomizationFunctions/MF_Base_CroppedEmissive}.Result
   Normal=@n42.RGB
  ) {attributeSetTypes [
    #1
    #2
    #3
    #4
    #5
    #6
   ]}
  MaterialFunctionCall(In=@n43) {materialFunction=&MF_TechArtOverrides_accessories}
 ]
}
